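# Post-process AnimBP for the FrostGalore large male head that dampens head and neck rotation and poses eyelash bones.
# Outside emotes and front end, it additively counter-rotates head and both neck bones by scaled ref-pose deltas, by default only 40% of yaw.
# It derives blink, squint and lid-down values from facial curves, subtracting to prioritize blink.
# Chained bone modifiers apply three fixed-rotation sets per eye, with no visible link to those values.
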AnimBlueprint /BRCosmetics/Characters/Player/Male/Large/Heads/M_LRG_FrostGalore_Head/Meshes/M_LRG_FrostGalore_PostProcess_AnimBP
target_skeleton=&/FrostGalore_Head_Export_Skeleton1
graphs [
 AnimGraph {nodes [
   n1 = "Transform (Modify) Bone - Bone: head"(
    ComponentPose=Local To Component(LocalPose=Use cached pose 'CR_Start'())
    Alpha=1
    Translation="0,0,0"
    Rotation=Get head rot() -> (head rot:Rotator)
    Scale="1,1,1"
   ) {rotationMode=BMM_Additive, rotationSpace=BCS_BoneSpace, boneToModify {boneName=head}, rotation=<rot 50 50 50>}
   n2 = Get neck_02 rot() -> (neck_02 rot:Rotator)
   n3 = Component To Local(ComponentPose="Transform (Modify) Bone - Bone: neck_01"(
     ComponentPose="Transform (Modify) Bone - Bone: neck_02"(ComponentPose=@n1, Alpha=1, Translation="0,0,0", Rotation=@n2, Scale="1,1,1") {rotationMode=BMM_Additive, rotationSpace=BCS_BoneSpace, boneToModify {boneName=neck_02}}
     Alpha=1
     Translation="0,0,0"
     Rotation=Get neck_01 rot() -> (neck_01 rot:Rotator)
     Scale="1,1,1"
    ) {rotationMode=BMM_Additive, rotationSpace=BCS_BoneSpace, boneToModify {boneName=neck_01}})
   n4 = Use cached pose 'CR_Start'()
   n5 = Get EmotingOrFrontEnd() -> (EmotingOrFrontEnd:bool)
   n6 = "Transform (Modify) Bone - Bone: dyn_eyelashes_01_l"(
    ComponentPose=Local To Component(LocalPose=Blend Poses by bool(BlendPose_0=@n4, BlendPose_1=@n3, bActiveValue=@n5))
    Alpha=1
    Translation="0,0,0"
    Rotation="0,0,82"
    Scale="1,1,1"
   ) {
    rotationMode=BMM_Replace
    rotationSpace=BCS_BoneSpace
    scaleMode=BMM_Additive
    scaleSpace=BCS_BoneSpace
    translationMode=BMM_Additive
    translationSpace=BCS_BoneSpace
    boneToModify {boneName=dyn_eyelashes_01_l}
    rotation {roll=82}
   }
   n7 = "Transform (Modify) Bone - Bone: dyn_eyelashes_02_l"(ComponentPose=@n6, Alpha=1, Translation="0,0,0", Rotation="0,0,90", Scale="1,1,1") {
    rotationMode=BMM_Replace
    rotationSpace=BCS_BoneSpace
    scaleMode=BMM_Additive
    scaleSpace=BCS_BoneSpace
    translationMode=BMM_Additive
    translationSpace=BCS_BoneSpace
    boneToModify {boneName=dyn_eyelashes_02_l}
    rotation {roll=90}
   }
   n8 = "Transform (Modify) Bone - Bone: dyn_eyelashes_03_l"(ComponentPose=@n7, Alpha=1, Translation="0,0,0", Rotation="0,0,110", Scale="1,1,1") {
    rotationMode=BMM_Replace
    rotationSpace=BCS_BoneSpace
    scaleMode=BMM_Additive
    scaleSpace=BCS_BoneSpace
    translationMode=BMM_Additive
    translationSpace=BCS_BoneSpace
    boneToModify {boneName=dyn_eyelashes_03_l}
    rotation {roll=110}
   }
   n9 = "Transform (Modify) Bone - Bone: dyn_eyelashes_04_l"(ComponentPose=@n8, Alpha=1, Translation="0,0,0", Rotation="-5,0,0", Scale="1,1,1") {
    rotationMode=BMM_Replace
    rotationSpace=BCS_BoneSpace
    scaleMode=BMM_Additive
    scaleSpace=BCS_BoneSpace
    translationMode=BMM_Additive
    translationSpace=BCS_BoneSpace
    boneToModify {boneName=dyn_eyelashes_04_l}
    rotation {pitch=-5}
   }
   n10 = "Transform (Modify) Bone - Bone: dyn_eyelashes_01_l"(ComponentPose=@n9, Alpha=1, Translation="0,0,0", Rotation="0,0,52", Scale="1,1,1") {
    rotationMode=BMM_Replace
    rotationSpace=BCS_BoneSpace
    scaleMode=BMM_Additive
    scaleSpace=BCS_BoneSpace
    translationMode=BMM_Additive
    translationSpace=BCS_BoneSpace
    boneToModify {boneName=dyn_eyelashes_01_l}
    rotation {roll=52}
   }
   n11 = "Transform (Modify) Bone - Bone: dyn_eyelashes_02_l"(ComponentPose=@n10, Alpha=1, Translation="0,0,0", Rotation="0,0,57", Scale="1,1,1") {
    rotationMode=BMM_Replace
    rotationSpace=BCS_BoneSpace
    scaleMode=BMM_Additive
    scaleSpace=BCS_BoneSpace
    translationMode=BMM_Additive
    translationSpace=BCS_BoneSpace
    boneToModify {boneName=dyn_eyelashes_02_l}
    rotation {roll=57}
   }
   n12 = "Transform (Modify) Bone - Bone: dyn_eyelashes_03_l"(ComponentPose=@n11, Alpha=1, Translation="0,0,0", Rotation="0,0,55", Scale="1,1,1") {
    rotationMode=BMM_Replace
    rotationSpace=BCS_BoneSpace
    scaleMode=BMM_Additive
    scaleSpace=BCS_BoneSpace
    translationMode=BMM_Additive
    translationSpace=BCS_BoneSpace
    boneToModify {boneName=dyn_eyelashes_03_l}
    rotation {roll=55}
   }
   n13 = "Transform (Modify) Bone - Bone: dyn_eyelashes_04_l"(ComponentPose=@n12, Alpha=1, Translation="0,0,0", Rotation="15,0,0", Scale="1,1,1") {
    rotationMode=BMM_Replace
    rotationSpace=BCS_BoneSpace
    scaleMode=BMM_Additive
    scaleSpace=BCS_BoneSpace
    translationMode=BMM_Additive
    translationSpace=BCS_BoneSpace
    boneToModify {boneName=dyn_eyelashes_04_l}
    rotation {pitch=15}
   }
   n14 = "Transform (Modify) Bone - Bone: dyn_eyelashes_01_l"(ComponentPose=@n13, Alpha=1, Translation="0,0,0", Rotation="0,0,82", Scale="1,1,1") {
    rotationMode=BMM_Replace
    rotationSpace=BCS_BoneSpace
    scaleMode=BMM_Additive
    scaleSpace=BCS_BoneSpace
    translationMode=BMM_Additive
    translationSpace=BCS_BoneSpace
    boneToModify {boneName=dyn_eyelashes_01_l}
    rotation {roll=82}
   }
   n15 = "Transform (Modify) Bone - Bone: dyn_eyelashes_02_l"(ComponentPose=@n14, Alpha=1, Translation="0,0,0", Rotation="0,0,90", Scale="1,1,1") {
    rotationMode=BMM_Replace
    rotationSpace=BCS_BoneSpace
    scaleMode=BMM_Additive
    scaleSpace=BCS_BoneSpace
    translationMode=BMM_Additive
    translationSpace=BCS_BoneSpace
    boneToModify {boneName=dyn_eyelashes_02_l}
    rotation {roll=90}
   }
   n16 = "Transform (Modify) Bone - Bone: dyn_eyelashes_03_l"(ComponentPose=@n15, Alpha=1, Translation="0,0,0", Rotation="0,0,110", Scale="1,1,1") {
    rotationMode=BMM_Replace
    rotationSpace=BCS_BoneSpace
    scaleMode=BMM_Additive
    scaleSpace=BCS_BoneSpace
    translationMode=BMM_Additive
    translationSpace=BCS_BoneSpace
    boneToModify {boneName=dyn_eyelashes_03_l}
    rotation {roll=110}
   }
   n17 = "Transform (Modify) Bone - Bone: dyn_eyelashes_04_l"(ComponentPose=@n16, Alpha=1, Translation="0,0,0", Rotation="-7,0,0", Scale="1,1,1") {
    rotationMode=BMM_Replace
    rotationSpace=BCS_BoneSpace
    scaleMode=BMM_Additive
    scaleSpace=BCS_BoneSpace
    translationMode=BMM_Additive
    translationSpace=BCS_BoneSpace
    boneToModify {boneName=dyn_eyelashes_04_l}
    rotation {pitch=-7}
   }
   n18 = "Transform (Modify) Bone - Bone: dyn_eyelashes_01_r"(ComponentPose=@n17, Alpha=1, Translation="0,0,0", Rotation="0,0,82", Scale="1,1,1") {
    rotationMode=BMM_Replace
    rotationSpace=BCS_BoneSpace
    scaleMode=BMM_Additive
    scaleSpace=BCS_BoneSpace
    translationMode=BMM_Additive
    translationSpace=BCS_BoneSpace
    boneToModify {boneName=dyn_eyelashes_01_r}
    rotation {roll=82}
   }
   n19 = "Transform (Modify) Bone - Bone: dyn_eyelashes_02_r"(ComponentPose=@n18, Alpha=1, Translation="0,0,0", Rotation="0,0,90", Scale="1,1,1") {
    rotationMode=BMM_Replace
    rotationSpace=BCS_BoneSpace
    scaleMode=BMM_Additive
    scaleSpace=BCS_BoneSpace
    translationMode=BMM_Additive
    translationSpace=BCS_BoneSpace
    boneToModify {boneName=dyn_eyelashes_02_r}
    rotation {roll=90}
   }
   n20 = "Transform (Modify) Bone - Bone: dyn_eyelashes_03_r"(ComponentPose=@n19, Alpha=1, Translation="0,0,0", Rotation="0,0,100", Scale="1,1,1") {
    rotationMode=BMM_Replace
    rotationSpace=BCS_BoneSpace
    scaleMode=BMM_Additive
    scaleSpace=BCS_BoneSpace
    translationMode=BMM_Additive
    translationSpace=BCS_BoneSpace
    boneToModify {boneName=dyn_eyelashes_03_r}
    rotation {roll=100}
   }
   n21 = "Transform (Modify) Bone - Bone: dyn_eyelashes_04_r"(ComponentPose=@n20, Alpha=1, Translation="0,0,0", Rotation="5,0,0", Scale="1,1,1") {
    rotationMode=BMM_Replace
    rotationSpace=BCS_BoneSpace
    scaleMode=BMM_Additive
    scaleSpace=BCS_BoneSpace
    translationMode=BMM_Additive
    translationSpace=BCS_BoneSpace
    boneToModify {boneName=dyn_eyelashes_04_r}
    rotation {pitch=5}
   }
   n22 = "Transform (Modify) Bone - Bone: dyn_eyelashes_01_r"(ComponentPose=@n21, Alpha=1, Translation="0,0,0", Rotation="0,0,52", Scale="1,1,1") {
    rotationMode=BMM_Replace
    rotationSpace=BCS_BoneSpace
    scaleMode=BMM_Additive
    scaleSpace=BCS_BoneSpace
    translationMode=BMM_Additive
    translationSpace=BCS_BoneSpace
    boneToModify {boneName=dyn_eyelashes_01_r}
    rotation {roll=52}
   }
   n23 = "Transform (Modify) Bone - Bone: dyn_eyelashes_02_r"(ComponentPose=@n22, Alpha=1, Translation="0,0,0", Rotation="0,0,57", Scale="1,1,1") {
    rotationMode=BMM_Replace
    rotationSpace=BCS_BoneSpace
    scaleMode=BMM_Additive
    scaleSpace=BCS_BoneSpace
    translationMode=BMM_Additive
    translationSpace=BCS_BoneSpace
    boneToModify {boneName=dyn_eyelashes_02_r}
    rotation {roll=57}
   }
   n24 = "Transform (Modify) Bone - Bone: dyn_eyelashes_03_r"(ComponentPose=@n23, Alpha=1, Translation="0,0,0", Rotation="0,0,55", Scale="1,1,1") {
    rotationMode=BMM_Replace
    rotationSpace=BCS_BoneSpace
    scaleMode=BMM_Additive
    scaleSpace=BCS_BoneSpace
    translationMode=BMM_Additive
    translationSpace=BCS_BoneSpace
    boneToModify {boneName=dyn_eyelashes_03_r}
    rotation {roll=55}
   }
   n25 = "Transform (Modify) Bone - Bone: dyn_eyelashes_04_r"(ComponentPose=@n24, Alpha=1, Translation="0,0,0", Rotation="-15,0,0", Scale="1,1,1") {
    rotationMode=BMM_Replace
    rotationSpace=BCS_BoneSpace
    scaleMode=BMM_Additive
    scaleSpace=BCS_BoneSpace
    translationMode=BMM_Additive
    translationSpace=BCS_BoneSpace
    boneToModify {boneName=dyn_eyelashes_04_r}
    rotation {pitch=-15}
   }
   n26 = "Transform (Modify) Bone - Bone: dyn_eyelashes_01_r"(ComponentPose=@n25, Alpha=1, Translation="0,0,0", Rotation="0,0,82", Scale="1,1,1") {
    rotationMode=BMM_Replace
    rotationSpace=BCS_BoneSpace
    scaleMode=BMM_Additive
    scaleSpace=BCS_BoneSpace
    translationMode=BMM_Additive
    translationSpace=BCS_BoneSpace
    boneToModify {boneName=dyn_eyelashes_01_r}
    rotation {roll=82}
   }
   n27 = "Transform (Modify) Bone - Bone: dyn_eyelashes_02_r"(ComponentPose=@n26, Alpha=1, Translation="0,0,0", Rotation="0,0,90", Scale="1,1,1") {
    rotationMode=BMM_Replace
    rotationSpace=BCS_BoneSpace
    scaleMode=BMM_Additive
    scaleSpace=BCS_BoneSpace
    translationMode=BMM_Additive
    translationSpace=BCS_BoneSpace
    boneToModify {boneName=dyn_eyelashes_02_r}
    rotation {roll=90}
   }
   n28 = "Transform (Modify) Bone - Bone: dyn_eyelashes_03_r"(ComponentPose=@n27, Alpha=1, Translation="0,0,0", Rotation="0,0,105", Scale="1,1,1") {
    rotationMode=BMM_Replace
    rotationSpace=BCS_BoneSpace
    scaleMode=BMM_Additive
    scaleSpace=BCS_BoneSpace
    translationMode=BMM_Additive
    translationSpace=BCS_BoneSpace
    boneToModify {boneName=dyn_eyelashes_03_r}
    rotation {roll=105}
   }
   n29 = "Transform (Modify) Bone - Bone: dyn_eyelashes_04_r"(ComponentPose=@n28, Alpha=1, Translation="0,0,0", Rotation="7,0,0", Scale="1,1,1") {
    rotationMode=BMM_Replace
    rotationSpace=BCS_BoneSpace
    scaleMode=BMM_Additive
    scaleSpace=BCS_BoneSpace
    translationMode=BMM_Additive
    translationSpace=BCS_BoneSpace
    boneToModify {boneName=dyn_eyelashes_04_r}
    rotation {pitch=7}
   }
   Output Pose(Result=Component To Local(ComponentPose=@n29))
   Save cached pose 'CR_Start'(Pose=Input Pose()) {cacheName=CR_Start}
  ]}
 EventGraph {nodes [
   Event BlueprintUpdateAnimation()
   > n1 = Sequence!(execute) -> (then_0, then_1, then_2:Exec, then_3:Exec)
   > Set EmotingOrFrontEnd(EmotingOrFrontEnd:bool=OR Boolean(A=Get bIsInFrontEnd(), B=Get bIsPlayingEmote())) -> (Output_Get:bool)
   n2 = GetOwnerSkeletalMeshComponent()
   n3 = BreakRotator(InRot=BreakTransform(InTransform=GetDeltaTransformFromRefPose(self=@n2, BoneName=head, BaseName=neck_02)).Rotation)
   n4 = Get Z percentage() -> (Z percentage:float)
   n5 = Get X percentage() -> (X percentage:float)
   n6 = Get Y percentage() -> (Y percentage:float)
   Set head rot(
    execute=@n1.then_1
    head rot:Rotator=MakeRotator(
     Roll=float * float!(A=@n3, B=-1, C:double=@n5) -> (ReturnValue)
     Pitch=float * float(A=@n3.Pitch, B=-1, C=@n6)
     Yaw=float * float(A=@n3.Yaw, B=-1, C=@n4)
    )
   ) -> (Output_Get:Rotator)
   n7 = GetOwnerSkeletalMeshComponent()
   n8 = BreakRotator(InRot=BreakTransform(InTransform=GetDeltaTransformFromRefPose(self=@n7, BoneName=neck_02, BaseName=neck_01)).Rotation)
   n9 = Get Z percentage()
   n10 = Get X percentage()
   n11 = Get Y percentage()
   > Set neck_02 rot(neck_02 rot:Rotator=MakeRotator(Roll=float * float(A=@n8, B=-1, C=@n10), Pitch=float * float(A=@n8.Pitch, B=-1, C=@n11), Yaw=float * float(A=@n8.Yaw, B=-1, C=@n9))) -> (Output_Get:Rotator)
   n12 = GetOwnerSkeletalMeshComponent()
   n13 = BreakRotator(InRot=BreakTransform(InTransform=GetDeltaTransformFromRefPose(self=@n12, BoneName=neck_01, BaseName=spine_05)).Rotation)
   n14 = Get X percentage()
   n15 = Get Z percentage()
   n16 = Get Y percentage()
   > Set neck_01 rot(neck_01 rot:Rotator=MakeRotator(Roll=float * float(A=@n13, B=-1, C=@n14), Pitch=float * float(A=@n13.Pitch, B=-1, C=@n16), Yaw=float * float(A=@n13.Yaw, B=-1, C=@n15))) -> (Output_Get:Rotator)
   n17 = GetAnimInstance(self=GetPartSkeletalMeshComponent())
   Set EyeBlink_l_01(execute=@n1.then_2, EyeBlink_l_01:double=MapRangeClamped(Value=GetCurveValue(self=@n17, CurveName=l_blink_pose), InRangeA=0.2, InRangeB=1)) -> (Output_Get:double)
   > Set EyeSquint_l(EyeSquint_l:double=MapRangeClamped(Value=GetCurveValue(self=@n17, CurveName=l_squint_inner_pose), InRangeA=0.5, InRangeB=2)) -> (Output_Get:double)
   > Set UpperLid_Down_l(UpperLid_Down_l:double=MapRangeClamped(Value=GetCurveValue(self=@n17, CurveName=l_upperLidDown_pose), InRangeA=0.2, InRangeB=1)) -> (Output_Get:double)
   > Set EyeBlink_r_01(EyeBlink_r_01:double=MapRangeClamped(Value=GetCurveValue(self=@n17, CurveName=r_blink_pose), InRangeA=0.2, InRangeB=1)) -> (Output_Get:double)
   > Set EyeSquint_r(EyeSquint_r:double=MapRangeClamped(Value=GetCurveValue(self=@n17, CurveName=r_squint_inner_pose), InRangeA=0.6, InRangeB=2)) -> (Output_Get:double)
   > Set UpperLid_Down_r(UpperLid_Down_r:double=MapRangeClamped(Value=GetCurveValue(self=@n17, CurveName=r_upperLidDown_pose), InRangeA=0.2, InRangeB=1)) -> (Output_Get:double)
   n18 = Get EyeBlink_r_01() -> (EyeBlink_r_01:double)
   Set EyeSquint_r(
    execute=@n1.then_3
    EyeSquint_r=float - float(A=Get EyeSquint_r() -> (EyeSquint_r:double), B=float * float(A=@n18, B=1.5) {is=title K2Node_PromotableOperator})
   )
   n19 = Get UpperLid_Down_r() -> (UpperLid_Down_r:double)
   > Set EyeSquint_r(EyeSquint_r=float - float(A=Get EyeSquint_r(), B=float * float(A=@n19, B=1.5)))
   > Set UpperLid_Down_r(UpperLid_Down_r=float - float(A=Get UpperLid_Down_r(), B=Get EyeBlink_r_01()))
   n20 = Get EyeBlink_l_01() -> (EyeBlink_l_01:double)
   > Set EyeSquint_l(EyeSquint_l=float - float(A=Get EyeSquint_l() -> (EyeSquint_l:double), B=float * float(A=@n20, B=1.5)))
   n21 = Get UpperLid_Down_l() -> (UpperLid_Down_l:double)
   > Set EyeSquint_l(EyeSquint_l=float - float(A=Get EyeSquint_l(), B=float * float(A=@n21, B=1.5)))
   > Set UpperLid_Down_l(UpperLid_Down_l=float - float(A=Get UpperLid_Down_l(), B=Get EyeBlink_l_01()))
   TryGetPawnOwner()
  ]}
]
variables [
 EmotingOrFrontEnd {type=boolean}
 EyeBlink_l_01 {type=number}
 EyeBlink_l_eyelashe03 {type=number}
 EyeBlink_r_01 {type=number}
 EyeSquint_l {type=number}
 EyeSquint_l_eyelashe03 {type=number}
 EyeSquint_r {type=number}
 head rot {type=Rotator}
 neck_01 rot {type=Rotator}
 neck_02 rot {type=Rotator}
 UpperLid_Down_l {type=number}
 UpperLid_Down_l_eyelashe03 {type=number}
 UpperLid_Down_r {type=number}
 X percentage {type=number}
 Y percentage {type=number}
 Z percentage {type=number, default=0.4}
]
